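# Niagara effect for the PolishedJade emote: four jade-colored mesh and sprite elements following skeletal mesh bones or sockets, with ribbon trails, GPU sparkles and mist.
# Partitioned particles map to the four attach points, and a shared scratch dynamic input reads the First–FourthSpriteOn user floats to scale each element's visibility.
# System timeline time fades everything out.
# Time of day adjusts color and mist opacity, mist fades with camera distance, and the BlowingParticles emitter is disabled.
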NiagaraSystem /BRCosmetics/Animation/Game/MainPlayer/Emotes/PolishedJade/FX/NS_Emote_PolishedJade
dependencies {
 usedDataInterfaces [
  &NiagaraDataInterfaceCollisionQuery
  &NiagaraDataInterfaceLandscape
  &NiagaraDataInterfaceCurve
  &NiagaraDataInterfaceArrayInt32
  &NiagaraDataInterfaceArrayFloat
  &NiagaraDataInterfaceCamera
  &NiagaraDataInterfaceSkeletalMesh
 ]
 usedDynamicInputs [&SetBoolByFloatComparison]
 usedModules [
  &self:SystemSpawnScript.SystemScriptSource.SystemScriptGraph.NiagaraNodeAssignment_0.SetVariables_#1
  &SystemState
  &Timeline
 ]
}
emitters {
 BlowingParticles {
  bEnabled=false
  simTarget=GPUComputeSim
  emitterUpdateScript [
   &EmitterState(Life Cycle Mode=::NewEnumerator1, UseLoopDelay=false)
   &SpawnRate(SpawnRate=15, Use Spawn Probability=false)
  ]
  particleSpawnScript [
   &InitializeParticle(
    Lifetime Max=15
    Lifetime Min=12
    Lifetime Mode=::NewEnumerator1
    Mass Max=1.5
    Mass Mode=::NewEnumerator3
    Sprite Rotation Mode=::NewEnumerator2
    Sprite Size Mode=::NewEnumerator1
    Uniform Sprite Size Max=8
    Uniform Sprite Size Min=4
    UsePositionOffset=false
    Write Material Random=true
   )
   &ShapeLocation(MaskSpawnGroup=false, Override Seed=false, Sphere Radius=300, UseOwnerScale=true)
   &InitialMeshOrientation(Use Rotation Vector=false)
  ]
  particleUpdateScript [
   &ParticleState()
   &GravityForce()
   &WindForce(
    Friction=0.8
    Offset By Time:Vector3f=<50 10 -125>
    Scale Min / Max:Vector2f=<0.275 1>
    Speed Min / Max:Vector2f=<0 500>
    Turbulence Scale Mode:ENiagaraWindTurbulenceContributionMode=::NewEnumerator1
    Wind Speed=<200 0 0>
    Wind Speed Scale=&RandomRangeFloat(Minimum=0.333, Maximum=1)
   )
   &AerodynamicDrag(
    Aerodynamic Drag:NiagaraFloat=&RandomRangeFloat(Minimum=0.3333, Maximum=1)
    Aerodynamic Rotational Drag=0.15
    Lift Contribution=0.75
    Pivot Offset Maximum:Vector3f=<5 0.5 1>
    Pivot Offset Minimum:Vector3f=<-5 -0.5 -1>
    Rotation Strength:NiagaraFloat=&RandomRangeFloat(Minimum=1.8, Maximum=3.2)
    Use Linear Drag=true
    Use Rotational Drag=true
   )
   &Collision(
    Age Colliding Particles=false
    Enable Rest State=false
    Friction=2
    Friction During a Bounce=0.1
    Kill Offscreen Particles=true
    KillOnCollision=false
    Module.EnableMaxCollisionCount=false
    Module.Randomize Collision Normal Vector=false
    Particle Radius Scale=0.1
    Restitution=0.05
   )
   &SolveForcesAndVelocity(Clamp Velocity=false, Limit Acceleration=false)
   &ScaleSpriteSize(Uniform Curve Sprite Scale=DataInterface() {propertyValues {Curve {keys [
        {arriveTangent=14.5835, interpMode=RCIM_Cubic, leaveTangent=14.5835, tangentMode=RCTM_User, time=0, value=0}
        {arriveTangent=1.1111, interpMode=RCIM_Cubic, leaveTangent=1.1111, time=0.1, value=1}
        {arriveTangent=-1.1111, interpMode=RCIM_Cubic, leaveTangent=-1.1111, time=0.9, value=1}
        {arriveTangent=-8.7113, interpMode=RCIM_Cubic, leaveTangent=-8.7113, tangentMode=RCTM_User, time=1, value=0}
       ]}}})
   &AlignSpriteToMeshOrientation(SetAlignment=true, SetFacing=true)
   /Niagara/Modules/Collision/AlignParticlesWithCollisionPlane(Allow Alignment on Y=false, Allow Alignment on Z=false, Rotational Dampening Power=1)
  ]
  rendererClasses [&NiagaraSpriteRendererProperties]
  renderers [{rendererClass=&NiagaraSpriteRendererProperties}]
 }
 BlowingParticles002 {
  emitterUpdateScript [
   &EmitterState(
    Inactive Response=::NewEnumerator2
    Life Cycle Mode=::NewEnumerator1
    Loop Behavior=::NewEnumerator1
    Loop Duration Mode=::NewEnumerator1
    UseLoopDelay=false
   )
   &SpawnPerFrame(Spawn Count=4, Use Spawn Probability=false) {enabled=false}
   &SpawnRate(SpawnRate=90, Use Spawn Probability=false)
  ]
  particleSpawnScript [
   &PartitionParticles(Number Of Partitions=4)
   &InitializeParticle(
    Color=&FN_Select_Color_from_Time_Of_Day(
     Morning=<rgba 0.0028 0.1354 0.0934 1>
     Day=<rgba 0.0028 0.1354 0.0934 1>
     Evening=<rgba 0.0028 0.1354 0.0934 1>
     Night=<rgba 0.0028 0.089 0.1354 1>
    )
    Lifetime=&SelectFloatFromArray(
     Direct Array Index=$Particles.PartitionParticles.Partition
     Float Selection Array=DataInterface() {propertyValues {FloatData [0.7,0.6,0.65,0.5]}}
    )
    Ribbon Width=35
    Ribbon Width Mode=::NewEnumerator1
    Sprite Rotation Mode=::NewEnumerator2
    UsePositionOffset=false
    Write Material Random=true
   )
   &SkeletalMeshLocation(
    Bone / Socket Index=$Particles.PartitionParticles.Partition
    Bone Sampling Mode=::NewEnumerator4
    MaskSpawnGroup=false
    Skeletal Mesh=$System.Skeletal Mesh
   )
  ]
  particleUpdateScript [
   &ParticleState()
   /SetVariables(Particles.AlphaScale:NiagaraFloat=&Multiply_Float(
     A=&Lerp_Float(A=1, B=0, Alpha=$System.Timeline.Time)
     B=&self:ScratchDynamicInput(
      FourthSpriteOn:NiagaraFloat=$User.FourthSpriteOn
      FirstSpriteOn:NiagaraFloat=$User.FirstSpriteOn
      SecondSpriteOn:NiagaraFloat=$User.SecondSpriteOn
      ThirdSpriteOn:NiagaraFloat=$User.ThirdSpriteOn
      UsePartitions:NiagaraBool=true
      Partition:NiagaraInt32=$Particles.PartitionParticles.Partition
     )
    )) {isSetParameters=true}
   &ScaleColor(
    Scale Alpha=&FloatFromCurve(
     FloatCurve=DataInterface() {propertyValues {Curve {keys [
         {arriveTangent=9.0725, interpMode=RCIM_Cubic, leaveTangent=9.0725, tangentMode=RCTM_User, time=0, value=0}
         {arriveTangent=0.1044, interpMode=RCIM_Cubic, leaveTangent=0.1044, tangentMode=RCTM_User, time=0.1408, value=0.9847}
         {arriveTangent=-1.146, interpMode=RCIM_Cubic, leaveTangent=-1.146, time=0.7413, value=0.7359}
         {arriveTangent=-0.0405, interpMode=RCIM_Cubic, leaveTangent=-0.0405, tangentMode=RCTM_User, time=1, value=0}
        ]}}}
     Scale Curve=$Particles.AlphaScale
    )
    Scale RGB=&VectorFromFloat(Value=2)
    ScaleA=true
    ScaleRGB=true
   )
   &CurlNoiseForce(Bias Noise=false, Noise Strength=50, Pan Noise=false)
   &SolveForcesAndVelocity(Clamp Velocity=false, Limit Acceleration=false)
   &ScaleRibbonWidth(Ribbon Width Scale=&FloatFromCurve(FloatCurve=DataInterface() {propertyValues {Curve {keys [{interpMode=RCIM_Cubic, tangentMode=RCTM_User, time=0, value=1}, {interpMode=RCIM_Cubic, tangentMode=RCTM_User, time=1, value=0.25}]}}}))
   &DynamicMaterialParameters(
    Index 0 Param 1=1
    Index 0 Param 2=1
    Index 0 Param 3=2
    Index 0 Param 4=-5
    Index 1 Param 1=1
    Index 1 Param 2=1
    Module.Param0WriteEnabled=true
    Module.Param0WriteEnabled1=true
    Module.Param1WriteEnabled=true
    Module.Param1WriteEnabled1=false
    Module.Param2WriteEnabled=true
    Module.Param2WriteEnabled1=false
    Module.Param3WriteEnabled=false
    Module.Param3WriteEnabled1=false
    Write Parameter Index 1=true
   ) {module=DynamicMaterialParameters001}
   /SetVariables(
    Particles.Color_02:LinearColor=&MultiplyLinearColor(
     Linear Color=$Particles.Color
     Scale Factor=&MakeFloat4FromVectorAndFloat("Vector (RGB)"=&VectorFromFloat(Value=5), "Float (Alpha)"=0.25)
    )
    Particles.Dyn2:Vector4f=<1 0.5 2 1>
    Particles.RWidth_02:NiagaraFloat=&Multiply_Float(A=$Particles.RibbonWidth, B=0.75)
   ) {isSetParameters=true}
   /SetVariables(
    Particles.Color_03:LinearColor=&MultiplyLinearColor(
     Linear Color=$Particles.Color
     Scale Factor=&MakeFloat4FromVectorAndFloat("Vector (RGB)"=&VectorFromFloat(Value=10), "Float (Alpha)"=0.5)
    )
    Particles.RWidth_03:NiagaraFloat=&Multiply_Float(A=$Particles.RibbonWidth, B=0.25)
   ) {isSetParameters=true}
  ]
  rendererClasses [&NiagaraSpriteRendererProperties, &NiagaraRibbonRendererProperties]
  renderers [
   {rendererClass=&NiagaraSpriteRendererProperties}
   {rendererClass=&NiagaraRibbonRendererProperties, rendererIndex=1}
   {rendererClass=&NiagaraRibbonRendererProperties, rendererIndex=2}
   {rendererClass=&NiagaraRibbonRendererProperties, rendererIndex=3}
  ]
 }
 Em_Sprite {
  emitterUpdateScript [
   &EmitterState(
    Inactive Response=::NewEnumerator2
    Life Cycle Mode=::NewEnumerator1
    Loop Behavior=::NewEnumerator1
    Loop Duration Mode=::NewEnumerator1
    UseLoopDelay=false
   )
   &SpawnBurst_Instantaneous(Spawn Count=4, Use Spawn Probability=false, UseLoopCountLimit=false)
  ]
  particleSpawnScript [&InitializeParticle(
    Color=&FN_Select_Color_from_Time_Of_Day(
     Morning=<rgba 0.0776 0.5729 0.4053 1>
     Day=<rgba 0.0776 0.5729 0.4053 1>
     Evening=<rgba 0.0776 0.5729 0.4053 1>
     Night=<rgba 0.0776 0.3665 0.5729 1>
    )
    Mesh Scale Mode=::NewEnumerator3
    Mesh Uniform Scale=&RandomRangeFloat(Minimum=0.9, Maximum=1.1)
    Sprite Size Mode=::NewEnumerator3
    Uniform Sprite Size=40
    UsePositionOffset=false
    Write Material Random=true
   )]
  particleUpdateScript [
   &ParticleState(Kill Particles When Lifetime Has Elapsed=false, Let Infinitely Lived Particles Die When Emitter Deactivates=true)
   /SetVariables(Particles.AlphaScale:NiagaraFloat=&Lerp_Float(A=1, B=0, Alpha=$System.Timeline.Time)) {isSetParameters=true}
   &ScaleColor(
    Scale Alpha=&Multiply_Float(
     A=$Particles.AlphaScale
     B=&self:ScratchDynamicInput(
      FourthSpriteOn:NiagaraFloat=$User.FourthSpriteOn
      FirstSpriteOn:NiagaraFloat=$User.FirstSpriteOn
      SecondSpriteOn:NiagaraFloat=$User.SecondSpriteOn
      ThirdSpriteOn:NiagaraFloat=$User.ThirdSpriteOn
      UsePartitions:NiagaraBool=false
      Partition:NiagaraInt32=0
     )
    )
    ScaleA=true
    ScaleRGB=true
   )
   &ScaleMeshSize(Scale Factor=&VectorFromFloat(Value=&FloatFromCurve(
      FloatCurve=DataInterface() {propertyValues {Curve {keys [{interpMode=RCIM_Cubic, tangentMode=RCTM_User, time=0, value=0}, {interpMode=RCIM_Cubic, tangentMode=RCTM_User, time=0.25, value=1}]}}}
      Scale Curve=&Multiply_Float(
       A=&self:ScratchDynamicInput(
        FourthSpriteOn:NiagaraFloat=$User.FourthSpriteOn
        FirstSpriteOn:NiagaraFloat=$User.FirstSpriteOn
        SecondSpriteOn:NiagaraFloat=$User.SecondSpriteOn
        ThirdSpriteOn:NiagaraFloat=$User.ThirdSpriteOn
        UsePartitions:NiagaraBool=false
        Partition:NiagaraInt32=0
       )
       B=$Particles.AlphaScale
      )
     )))
   &SkeletalMeshLocation(
    Bone / Socket Index=&ReturnExecIndex()
    Bone Sampling Mode=::NewEnumerator4
    MaskSpawnGroup=false
    Skeletal Mesh=$System.Skeletal Mesh
   )
   /SetVariables(Particles.SpriteDynamic:Vector4f=<3 1 0 0>) {isSetParameters=true}
   &FN_Sprite_Attributes(
    Color=&MultiplyLinearColor(
     Linear Color=$Particles.Color
     Scale Factor=&MakeFloat4FromVectorAndFloat(
      "Vector (RGB)"=<0 0 0>
      "Float (Alpha)"=&FloatFromCurve(
       FloatCurve=DataInterface() {propertyValues {Curve {keys [{interpMode=RCIM_Cubic, tangentMode=RCTM_User, time=0, value=0}, {interpMode=RCIM_Cubic, tangentMode=RCTM_User, time=1, value=1}]}}}
       Scale Curve=0.75
      )
     )
    )
    Create Color Attribute=true
    Create SpriteSize Attribute=true
    SpriteSize=&Multiply_Vector2D_ByFloat(Float=1.5, Vector2D=$Particles.SpriteSize)
    Write to Module Namespace=true
   ) {module=FN_Sprite_Attributes001}
   &FN_Sprite_Attributes(
    Color=&MultiplyLinearColor(
     Linear Color=$Particles.Color
     Scale Factor=&MakeFloat4FromVectorAndFloat(
      "Vector (RGB)"=<1 1 1>
      "Float (Alpha)"=&FloatFromCurve(FloatCurve=DataInterface() {propertyValues {Curve {keys [{interpMode=RCIM_Cubic, tangentMode=RCTM_User, time=0, value=0}, {interpMode=RCIM_Cubic, tangentMode=RCTM_User, time=1, value=1}]}}})
     )
    )
    Create Color Attribute=true
    Write to Module Namespace=true
   )
   /SetVariables(Particles.VisibilityTag=&SelectIntFromArray(
     Direct Array Index=&ReturnExecIndex()
     Int Selection Array=DataInterface() {propertyValues {IntData [0,0,1,0]}}
    )) {enabled=false, isSetParameters=true}
  ]
  rendererClasses [&NiagaraMeshRendererProperties, &NiagaraSpriteRendererProperties]
  renderers [
   {rendererClass=&NiagaraMeshRendererProperties, rendererIndex=0}
   {rendererClass=&NiagaraSpriteRendererProperties, rendererIndex=1}
   {rendererClass=&NiagaraSpriteRendererProperties, rendererIndex=2}
   {rendererClass=&NiagaraMeshRendererProperties, rendererIndex=3}
   {rendererClass=&NiagaraSpriteRendererProperties, rendererIndex=4}
   {rendererClass=&NiagaraSpriteRendererProperties, rendererIndex=5}
  ]
 }
 NE_Sparkles {
  simTarget=GPUComputeSim
  emitterUpdateScript [
   &EmitterState(Inactive Response=::NewEnumerator2, Life Cycle Mode=::NewEnumerator1, Loop Duration=1, UseLoopDelay=false)
   &SpawnRate(SpawnRate=120, Use Spawn Probability=false)
  ]
  particleSpawnScript [
   &PartitionParticles(Number Of Partitions=4)
   &InitializeParticle(
    Color=&RandomRangeLinearColor(Minimum=<rgba 0.0018 0.0885 0.0857 1>, Maximum=<rgba 0.2203 0.9 0.7754 1>)
    Lifetime=&RandomRangeFloat(Minimum=0.25, Maximum=1.25)
    Sprite Rotation Mode=::NewEnumerator2
    Sprite Size Mode=::NewEnumerator3
    Uniform Sprite Size=5
    UsePositionOffset=false
    Write Material Random=true
   )
   &SkeletalMeshLocation(
    Bone / Socket Index=$Particles.PartitionParticles.Partition
    Bone Sampling Mode=::NewEnumerator4
    MaskSpawnGroup=false
    Skeletal Mesh=$System.Skeletal Mesh
   )
   &ShapeLocation(MaskSpawnGroup=false, Non Uniform Scale=<1 0 2>, Override Seed=false, Sphere Radius=10, UseOwnerScale=true)
   /SetVariables(Particles.VisibilityTag=&MakeIntFromFloat(Float=&FloatFromCurve(
      FloatCurve=DataInterface() {propertyValues {Curve {keys [<at 0 0.5 RCIM_Constant>,<at 0.95 1.5 RCIM_Constant>]}}}
      CurveIndex=&RandomRangeFloat(Minimum=0, Maximum=1)
     ))) {isSetParameters=true}
  ]
  particleUpdateScript [
   &ParticleState()
   /SetVariables(Particles.AlphaScale:NiagaraFloat=&Lerp_Float(A=1, B=0, Alpha=$System.Timeline.Time)) {isSetParameters=true}
   &SpriteRotationRate(Rotation Rate=&Multiply_Float(
     A=&UniformAOrBFloat(A=-1, Evaluation Type=::NewEnumerator0, Recalculate Random Each Loop=true, Override Seed=false)
     B=&RandomRangeFloat(Minimum=240, Maximum=480)
    ))
   &ScaleSpriteSize(
    Scale Sprite Size Mode=::NewEnumerator0
    Uniform Scale Factor=&FloatFromCurve(FloatCurve=DataInterface() {propertyValues {Curve {keys [
         {arriveTangent=-0.0089, interpMode=RCIM_Cubic, leaveTangent=-0.0089, tangentMode=RCTM_User, time=0, value=0}
         {arriveTangent=-2.5249, interpMode=RCIM_Cubic, leaveTangent=-2.5249, tangentMode=RCTM_User, time=0.1408, value=1}
         {arriveTangent=-0.0405, interpMode=RCIM_Cubic, leaveTangent=-0.0405, tangentMode=RCTM_User, time=1, value=0}
        ]}}})
   )
   &ScaleColor(
    Scale Alpha=&FloatFromCurve(
     FloatCurve=DataInterface() {propertyValues {Curve {keys [
         {arriveTangent=0.088, interpMode=RCIM_Cubic, leaveTangent=0.088, tangentMode=RCTM_User, time=0, value=0}
         {arriveTangent=-2.5249, interpMode=RCIM_Cubic, leaveTangent=-2.5249, tangentMode=RCTM_User, time=0.1408, value=0.9847}
         {arriveTangent=-0.0405, interpMode=RCIM_Cubic, leaveTangent=-0.0405, tangentMode=RCTM_User, time=1, value=0}
        ]}}}
     Scale Curve=&Multiply_Float(
      A=$Particles.AlphaScale
      B=&self:ScratchDynamicInput(
       FourthSpriteOn:NiagaraFloat=$User.FourthSpriteOn
       FirstSpriteOn:NiagaraFloat=$User.FirstSpriteOn
       SecondSpriteOn:NiagaraFloat=$User.SecondSpriteOn
       ThirdSpriteOn:NiagaraFloat=$User.ThirdSpriteOn
       UsePartitions:NiagaraBool=true
       Partition:NiagaraInt32=$Particles.PartitionParticles.Partition
      )
     )
    )
    Scale RGB=&VectorFromFloat(Value=&Add_Float(
      A=&FloatFromCurve(FloatCurve=DataInterface() {propertyValues {Curve {keys [
           {arriveTangent=-0.0052, interpMode=RCIM_Cubic, leaveTangent=-0.0052, tangentMode=RCTM_User, time=0, value=1}
           {arriveTangent=-2.5249, interpMode=RCIM_Cubic, leaveTangent=-2.5249, tangentMode=RCTM_User, time=0.1408, value=3.5}
           {interpMode=RCIM_Cubic, leaveTangent=-0.0405, time=1, value=1}
          ]}}})
      B=&FloatFromCurve(
       FloatCurve=DataInterface() {propertyValues {Curve {keys [
           {arriveTangent=-0.0052, interpMode=RCIM_Cubic, leaveTangent=-0.0052, tangentMode=RCTM_User, time=0.4, value=0}
           {arriveTangent=6292.0376, interpMode=RCIM_Cubic, leaveTangent=-4742.8989, tangentMode=RCTM_Break, time=0.5, value=350}
           {interpMode=RCIM_Cubic, leaveTangent=6.25, time=0.6, value=0}
          ]}}}
       CurveIndex=&ModuloFloat(
        A=1
        B=&Add_Float(A=$Particles.NormalizedAge, B=&RandomRangeFloat(Minimum=0, Maximum=1))
       )
      )
     ))
    ScaleA=true
    ScaleRGB=true
   )
   &CurlNoiseForce(Bias Noise=false, Noise Strength=150, Pan Noise=false)
   &SolveForcesAndVelocity(Clamp Velocity=false, Limit Acceleration=false)
   &FN_Sprite_Attributes(
    Create SpriteSize Attribute=true
    SpriteSize=&Multiply_Vector2D_ByFloat(Float=1.5, Vector2D=$Particles.SpriteSize)
    Write to Module Namespace=true
   )
  ]
  rendererClasses [&NiagaraSpriteRendererProperties]
  renderers [
   {rendererClass=&NiagaraSpriteRendererProperties}
   {rendererClass=&NiagaraSpriteRendererProperties, rendererIndex=1}
  ]
 }
 Nanomist001 {
  emitterUpdateScript [
   &EmitterState(
    Inactive Response=::NewEnumerator2
    Life Cycle Mode=::NewEnumerator1
    Loop Behavior=::NewEnumerator1
    Loop Duration Mode=::NewEnumerator1
    UseLoopDelay=false
   )
   &SpawnBurst_Instantaneous(
    Spawn Count=&Multiply_Int(A=&RandomRangeInt(Minimum=8, Maximum=10), B=2)
    Use Spawn Probability=false
    UseLoopCountLimit=false
   ) {module=SpawnBurst_Instantaneous001}
  ]
  particleSpawnScript [
   &PartitionParticles(Number Of Partitions=4)
   &InitializeParticle(
    Color Maximum=<rgba 0.2203 0.9 0.7754 1>
    Color Minimum=<rgba 0.0018 0.0885 0.0857 1>
    Color Mode=::NewEnumerator2
    Lifetime Max=0.8
    Lifetime Min=0.6
    Lifetime Mode=::NewEnumerator1
    Sprite Rotation Mode=::NewEnumerator2
    Sprite Size Mode=::NewEnumerator1
    Uniform Sprite Size Max=80
    Uniform Sprite Size Min=60
    UsePositionOffset=false
    Write Material Random=true
   )
   &AddVelocity(
    Cone Axis=&TransformVector(Vector=<0 0 1>, Source Space=::World, Destination Space=::Local)
    Module.UseConeFalloff=false
    Module.UseVelDistribution=false
    Override Seed=false
    Velocity Mode=::NewEnumerator2
    Velocity Speed=&RandomRangeFloat(Minimum=200, Maximum=300)
   ) {enabled=false}
   &SkeletalMeshLocation(
    Bone / Socket Index=$Particles.PartitionParticles.Partition
    Bone Sampling Mode=::NewEnumerator4
    MaskSpawnGroup=false
    Sampled Position Offset=<0 0 5>
    Skeletal Mesh=$System.Skeletal Mesh
    Velocity Sampling=::NewEnumerator0
    VelocitySpeedLimit=false
   ) {enabled=false}
   &ShapeLocation(
    Hemisphere Distribution=<1 0.5>
    MaskSpawnGroup=false
    Override Seed=false
    Rotation Angle=90
    Rotation Mode=::NewEnumerator1
    Sphere Radius=5
    UseOwnerScale=true
   )
   &AddVelocity(
    Cone Axis=<0 0 -1>
    Module.UseConeFalloff=false
    Module.UseVelDistribution=false
    Override Seed=false
    Velocity Mode=::NewEnumerator2
    Velocity Speed=&RandomRangeFloat(Minimum=25, Maximum=250)
   ) {enabled=false, module=AddVelocity001}
   &CurlNoiseForce(Noise Strength=100, Pan Noise=false) {enabled=false, module=CurlNoiseForce001}
   &ApplyInitialForces() {enabled=false}
   &AddVelocity(Velocity=&Multiply_VectorByFloat(Vector=$Particles.ShapeLocation.ShapeVector, Float=20)) {enabled=false, module=AddVelocity002}
  ]
  particleUpdateScript [
   &ParticleState(Kill Particles When Lifetime Has Elapsed=false, Let Infinitely Lived Particles Die When Emitter Deactivates=true)
   /SetVariables(Particles.AlphaScale:NiagaraFloat=&Lerp_Float(A=1, B=0, Alpha=$System.Timeline.Time)) {isSetParameters=true}
   &SpriteRotationRate(Rotation Rate=&RandomRangeFloat(Minimum=20, Maximum=30))
   &ScaleSpriteSize(
    Uniform Curve Index=&VectorLength(Vector=&Subtract_Position(
      A=$Particles.Position
      B=&ReturnCameraPosition(Coordinate Space:ENiagaraCoordinateSpace=::Simulation, Camera Query:NiagaraDataInterfaceCamera=DataInterface())
     ))
    Uniform Curve Sprite Scale=DataInterface() {propertyValues {Curve {keys [
        {interpMode=RCIM_Cubic, tangentMode=RCTM_Break, time=200, value=1}
        {arriveTangent=0.0003, interpMode=RCIM_Cubic, tangentMode=RCTM_Break, time=400, value=1.1}
       ]}}}
   )
   &DynamicMaterialParameters(
    Index 0 Param 1=&RandomRangeFloat(Minimum=0, Maximum=100)
    Module.Param0WriteEnabled=true
    Module.Param1WriteEnabled=true
    Module.Param2WriteEnabled=true
    Module.Param3WriteEnabled=true
   )
   &Drag(Drag=2, Use Linear Drag=true, Use Rotational Drag=true)
   &CurlNoiseForce(Noise Frequency=20, Noise Strength=50, Pan Noise=true, Pan Noise Field=<0 0 -1>)
   &GravityForce(Gravity=&RandomRangeVector(Minimum=<0 0 -150>, Maximum=<0 0 -100>))
   &SolveForcesAndVelocity(Clamp Velocity=false, Limit Acceleration=false)
   &ScaleColor(
    Scale Alpha=&FloatFromCurve(
     FloatCurve=DataInterface() {propertyValues {
       bExposeCurve=true
       bOptimizeLUT=false
       Curve {keys [{interpMode=RCIM_Cubic, tangentMode=RCTM_User, time=0, value=0}, {interpMode=RCIM_Cubic, tangentMode=RCTM_User, time=1, value=1}]}
      }}
     Scale Curve=&Multiply_Float(
      A=&Multiply_Float(
       A=2
       B=&self:ScratchDynamicInput(
        FourthSpriteOn:NiagaraFloat=$User.FourthSpriteOn
        FirstSpriteOn:NiagaraFloat=$User.FirstSpriteOn
        SecondSpriteOn:NiagaraFloat=$User.SecondSpriteOn
        ThirdSpriteOn:NiagaraFloat=$User.ThirdSpriteOn
        UsePartitions:NiagaraBool=true
        Partition:NiagaraInt32=$Particles.PartitionParticles.Partition
       )
      )
      B=$Particles.AlphaScale
     )
    )
    Scale RGB=&VectorFromFloat(Value=3)
    ScaleA=true
    ScaleRGB=true
   )
   &ScaleColor(
    Scale Alpha=&FloatFromCurve(
     FloatCurve=DataInterface() {propertyValues {Curve {keys [
         {interpMode=RCIM_Cubic, tangentMode=RCTM_Break, time=400, value=1}
         {arriveTangent=0.0003, interpMode=RCIM_Cubic, tangentMode=RCTM_Break, time=700, value=0}
        ]}}}
     CurveIndex=&VectorLength(Vector=&Subtract_Position(
       A=$Particles.Position
       B=&ReturnCameraPosition(Coordinate Space:ENiagaraCoordinateSpace=::Simulation, Camera Query:NiagaraDataInterfaceCamera=DataInterface())
      ))
    )
    ScaleA=true
    ScaleRGB=true
   ) {module=ScaleColor001}
   &ScaleColor(
    Scale Alpha=&FNTimeOfDayScalar(Night Scale=1, Morning Scale=0.4, Evening Scale=0.5)
    ScaleA=true
    ScaleRGB=true
   ) {module=ScaleColor002}
   &CameraOffset(Camera Offset Amount=&RandomRangeFloat(Minimum=-10, Maximum=-30))
   &SkeletalMeshLocation(
    Bone / Socket Index=$Particles.PartitionParticles.Partition
    Bone Sampling Mode=::NewEnumerator4
    MaskSpawnGroup=false
    Sampled Position Offset=<0 0 5>
    Skeletal Mesh=$System.Skeletal Mesh
   ) {module=SkeletalMeshLocation001}
  ]
  rendererClasses [&NiagaraSpriteRendererProperties]
  renderers [{rendererClass=&NiagaraSpriteRendererProperties}]
 }
}
summary {
 systemName=NS_Emote_PolishedJade
 emitters [
  {bEnabled=false, emitterName=BlowingParticles, simTarget=GPUComputeSim, rendererClasses [&NiagaraSpriteRendererProperties]}
  {emitterName=Em_Sprite, rendererClasses [&NiagaraMeshRendererProperties, &NiagaraSpriteRendererProperties]}
  {emitterName=NE_Sparkles, simTarget=GPUComputeSim, rendererClasses [&NiagaraSpriteRendererProperties]}
  {
   emitterName=BlowingParticles002
   rendererClasses [&NiagaraSpriteRendererProperties, &NiagaraRibbonRendererProperties]
  }
  {emitterName=Nanomist001, rendererClasses [&NiagaraSpriteRendererProperties]}
 ]
 userVariables [
  User.bIsFrontEnd {type=NiagaraBool}
  User.FirstSpriteOn {type=NiagaraFloat, defaultValue=0, description=""}
  User.FourthSpriteOn {type=NiagaraFloat, defaultValue=0, description=""}
  User.Interpolate Intensity {type=NiagaraFloat, defaultValue=0.1, description=""}
  User.NormalizedNotifyProgress {type=NiagaraFloat, defaultValue=0, description=""}
  User.RibbonIDOffset {type=NiagaraInt32}
  User.SecondSpriteOn {type=NiagaraFloat, defaultValue=0, description=""}
  User.ThirdSpriteOn {type=NiagaraFloat, defaultValue=0, description=""}
  User.TurnOff {type=NiagaraFloat, defaultValue=0, description=""}
 ]
}
