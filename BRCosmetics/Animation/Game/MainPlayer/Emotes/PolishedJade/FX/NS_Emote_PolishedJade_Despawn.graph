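# Niagara despawn effect for the PolishedJade emote: a short teal burst of glow, smoke and sparks on a skeletal mesh.
# All three emitters split particles into four partitions, each mapped to a bone or socket index.
# Glow sprites re-sample that position every update, while smoke and noise-driven sparks inherit quickly fading mesh velocity.
# Glow and sparks simulate on GPU; smoke sets a dynamic material parameter over its lifetime.
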NiagaraSystem /BRCosmetics/Animation/Game/MainPlayer/Emotes/PolishedJade/FX/NS_Emote_PolishedJade_Despawn
dependencies {
 usedDataInterfaces [
  &NiagaraDataInterfaceCurve
  &NiagaraDataInterfaceArrayFloat3
  &NiagaraDataInterfaceColorCurve
  &NiagaraDataInterfaceVectorCurve
  &NiagaraDataInterfaceSkeletalMesh
 ]
 usedDynamicInputs [&SetBoolByFloatComparison]
 usedModules [
  &self:SystemSpawnScript.SystemScriptSource.SystemScriptGraph.NiagaraNodeAssignment_0.SetVariables_#1
  &self:SystemSpawnScript.SystemScriptSource.SystemScriptGraph.NiagaraNodeAssignment_1.SetVariables_#2
  &SystemState
  &Timeline
 ]
}
emitters {
 BlowingParticles001 {
  simTarget=GPUComputeSim
  emitterUpdateScript [
   &EmitterState(Life Cycle Mode=::NewEnumerator1, Loop Behavior=::NewEnumerator1, Loop Duration=1, UseLoopDelay=false)
   &SpawnBurst_Instantaneous(Spawn Count=4, Use Spawn Probability=false, UseLoopCountLimit=false)
   &SpawnRate(SpawnRate=15, Use Spawn Probability=false) {enabled=false}
  ]
  particleSpawnScript [
   &InitializeParticle(
    Color=<rgba 0.2203 0.9 0.7754 1>
    Lifetime=&RandomRangeFloat(Minimum=0.25, Maximum=0.35)
    Mass Max=1.5
    Mass Mode=::NewEnumerator3
    Sprite Rotation Mode=::NewEnumerator2
    Sprite Size Mode=::NewEnumerator3
    Uniform Sprite Size=&RandomRangeFloat(Minimum=70, Maximum=75)
    UsePositionOffset=false
    Write Material Random=true
   )
   &PartitionParticles(Number Of Partitions=4)
  ]
  particleUpdateScript [
   &ParticleState()
   &SkeletalMeshLocation(
    Bone / Socket Index=$Particles.PartitionParticles.Partition
    Bone Sampling Mode=::NewEnumerator4
    MaskSpawnGroup=false
    Sampled Position Offset=<0 0 5>
    Skeletal Mesh=$System.Skeletal Mesh
   )
   &ScaleSpriteSize(Uniform Curve Sprite Scale=DataInterface() {propertyValues {Curve {keys [
        {arriveTangent=0.0428, interpMode=RCIM_Cubic, leaveTangent=0.0428, tangentMode=RCTM_User, time=0, value=0}
        {arriveTangent=0.0649, interpMode=RCIM_Cubic, leaveTangent=0.0649, tangentMode=RCTM_User, time=1, value=1}
       ]}}})
   &ScaleColor(
    Scale Alpha=&FloatFromCurve(FloatCurve=DataInterface() {propertyValues {Curve {keys [
         {arriveTangent=9.0725, interpMode=RCIM_Cubic, leaveTangent=9.0725, tangentMode=RCTM_User, time=0, value=0}
         {arriveTangent=-2.5249, interpMode=RCIM_Cubic, leaveTangent=-2.5249, tangentMode=RCTM_User, time=0.1408, value=0.9847}
         {arriveTangent=-0.0405, interpMode=RCIM_Cubic, leaveTangent=-0.0405, tangentMode=RCTM_User, time=1, value=0}
        ]}}})
    Scale RGB=&VectorFromFloat(Value=5)
    ScaleA=true
    ScaleRGB=true
   )
   &FN_Sprite_Attributes(
    Color=&MultiplyLinearColor(
     Linear Color=$Particles.Color
     Scale Factor=&MakeFloat4FromVectorAndFloat("Vector (RGB)"=&VectorFromFloat(Value=0.05), "Float (Alpha)"=1)
    )
    Create Color Attribute=true
    Create SpriteSize Attribute=true
    SpriteSize=&Multiply_Vector2D_ByFloat(Float=8, Vector2D=$Particles.SpriteSize)
    Write to Module Namespace=true
   )
  ]
  rendererClasses [&NiagaraSpriteRendererProperties]
  renderers [
   {rendererClass=&NiagaraSpriteRendererProperties}
   {rendererClass=&NiagaraSpriteRendererProperties, rendererIndex=1}
   {rendererClass=&NiagaraSpriteRendererProperties, rendererIndex=2}
  ]
 }
 Default_Pickaxe_Impact_Smoke {
  emitterUpdateScript [
   &EmitterState(Inactive Response=::NewEnumerator1, Life Cycle Mode=::NewEnumerator1, Loop Behavior=::NewEnumerator1, Loop Duration=2, UseLoopDelay=false)
   &SpawnBurst_Instantaneous(
    Spawn Count=&Multiply_Int(A=&RandomRangeInt(Minimum=3, Maximum=4), B=4)
    Use Spawn Probability=false
    UseLoopCountLimit=false
   )
  ]
  particleSpawnScript [
   &PartitionParticles(Number Of Partitions=4)
   &InitializeParticle(
    Color Maximum=<rgba 0.2203 0.9 0.5488 1>
    Color Minimum=<rgba 0.0018 0.0885 0.0698 1>
    Color Mode=::NewEnumerator2
    Lifetime Max=0.5
    Lifetime Min=0.35
    Lifetime Mode=::NewEnumerator1
    Sprite Size Mode=::NewEnumerator3
    Uniform Sprite Size=&Multiply_Float(A=0.45, B=&RandomRangeFloat(Minimum=100, Maximum=130))
    UsePositionOffset=false
    Write Material Random=true
   )
   &SkeletalMeshLocation(
    Bone / Socket Index=$Particles.PartitionParticles.Partition
    Bone Sampling Mode=::NewEnumerator4
    MaskSpawnGroup=false
    Orientation Sampling=::NewEnumerator1
    Sampled Position Offset=<0 0 5>
    Skeletal Mesh=$System.Skeletal Mesh
   )
   &CameraOffset(Camera Offset Amount=&RandomRangeFloat(Minimum=40, Maximum=50, Evaluation Type=::NewEnumerator1))
   &ShapeLocation(
    Hemisphere Distribution=<1 0.5>
    MaskSpawnGroup=false
    Override Seed=false
    Rotation Angle=90
    Rotation Mode=::NewEnumerator1
    Sphere Radius=15
    UseOwnerScale=true
   )
   &AddVelocity(Velocity=&Multiply_VectorByFloat(Vector=$Particles.ShapeLocation.ShapeVector, Float=20)) {enabled=false}
  ]
  particleUpdateScript [
   &ParticleState()
   &ScaleSpriteSize(Uniform Curve Sprite Scale=DataInterface() {propertyValues {Curve {keys [
        {arriveTangent=1.7761, interpMode=RCIM_Cubic, leaveTangent=1.7761, tangentMode=RCTM_User, time=0, value=0}
        {interpMode=RCIM_Cubic, tangentMode=RCTM_User, time=1, value=1}
       ]}}})
   &ScaleColor(
    Scale Alpha=&FloatFromCurve(FloatCurve=DataInterface() {propertyValues {Curve {keys [
         {arriveTangent=2.4358, interpMode=RCIM_Cubic, leaveTangent=2.4358, tangentMode=RCTM_User, time=0, value=0}
         <at 0.25 0.5 RCIM_Cubic>
         <at 1 0 RCIM_Cubic>
        ]}}})
    Scale RGB=&VectorFromCurve(VectorCurve=DataInterface() {propertyValues {XCurve {keys [<at 0 1>,<at 1 1>]}, YCurve {keys [<at 0 1>,<at 1 1>]}, ZCurve {keys [<at 0 1>,<at 1 1>]}}})
    ScaleA=true
    ScaleRGB=true
   )
   &DynamicMaterialParameters(
    Index 0 Param 1=&Add_Float(
     A=&RandomRangeFloat(Minimum=0, Maximum=1)
     B=&FloatFromCurve(
      FloatCurve=DataInterface() {propertyValues {Curve {keys [
          <at 0 0 RCIM_Cubic>
          {arriveTangent=1.5947, interpMode=RCIM_Cubic, leaveTangent=1.5947, tangentMode=RCTM_User, time=0.2557, value=0.5257}
          <at 1 1 RCIM_Cubic>
         ]}}}
      CurveIndex=$Particles.Age
      Scale Curve=0.25
     )
    )
    Module.Param0WriteEnabled=true
    Module.Param1WriteEnabled=true
    Module.Param2WriteEnabled=true
    Module.Param3WriteEnabled=true
   ) {module=DynamicMaterialParameters001}
   &SkeletalMeshLocation(
    Bone / Socket Index=$Particles.PartitionParticles.Partition
    Bone Sampling Mode=::NewEnumerator4
    Initialize Particle Position To Sampled Position=false
    MaskSpawnGroup=false
    Position Sampling=::NewEnumerator1
    Skeletal Mesh=$System.Skeletal Mesh
    Source Movement Scale:Vector3f=<0.95 0.95 0.95>
   ) {module=SkeletalMeshLocation001}
   &Drag(Drag=&RandomRangeFloat(Minimum=4, Maximum=4.5), Use Linear Drag=true, Use Rotational Drag=false)
   &AccelerationForce(Acceleration=&Multiply_Vector(
     A=&Multiply_VectorByFloat(
      Vector=&SubtractVectorFromPosition(Position=<0 0 0>, Vector=Linked() {linkedVariable {name=Particles.Position, type=NiagaraPosition}})
      Float=-10
     )
     B=<1 1 0>
    )) {enabled=false}
   &SolveForcesAndVelocity(Clamp Velocity=false, Limit Acceleration=false)
   /SetVariables(Particles.Velocity=&Add_Vector(
     A=$Particles.Presolve.Velocity
     B=&Multiply_Vector(
      A=&Multiply_VectorByFloat(
       Vector=$Particles.SkeletalMeshLocation001.SampledVelocity
       Float=&FloatFromCurve(FloatCurve=DataInterface() {propertyValues {Curve {keys [<at 0 1>,<at 0.15 0>]}}}, Scale Curve=0.12)
      )
      B=<0 0 1>
     )
    )) {isSetParameters=true}
  ]
  rendererClasses [&NiagaraSpriteRendererProperties]
  renderers [
   {rendererClass=&NiagaraSpriteRendererProperties}
   {rendererClass=&NiagaraSpriteRendererProperties, rendererIndex=1}
  ]
 }
 NE_FN_Sparks {
  simTarget=GPUComputeSim
  emitterUpdateScript [
   &EmitterState(Life Cycle Mode=::NewEnumerator1, Loop Behavior=::NewEnumerator1, Loop Duration=1, UseLoopDelay=false)
   &SpawnRate(SpawnRate=200, Use Spawn Probability=false) {enabled=false}
   &SpawnBurst_Instantaneous(Spawn Count=120, Spawn Time=0.02, Use Spawn Probability=false)
  ]
  particleSpawnScript [
   /SetVariables(Particles.SparkIntensity:NiagaraFloat=&FloatFromCurve(
     FloatCurve=DataInterface() {propertyValues {Curve {keys [
         <at 0 1 RCIM_Cubic>
         {arriveTangent=-2.5317, interpMode=RCIM_Cubic, leaveTangent=-2.5317, tangentMode=RCTM_User, time=0.6, value=0.5}
         <at 1 0 RCIM_Cubic>
        ]}}}
     CurveIndex=&RandomRangeFloat(Minimum=0, Maximum=1)
    )) {isSetParameters=true}
   &PartitionParticles(Number Of Partitions=4)
   &FN_SetSpawnGroup(Number of Spawn Groups=3) {enabled=false}
   &InitializeParticle(
    Color=&LerpLinearColors(StartColor=<rgba 0.0005 0.026 0.0252 1>, EndColor=<rgba 0.2203 0.9 0.7754 1>, LerpFactor=$Particles.SparkIntensity)
    Lifetime Max=0.95
    Lifetime Min=0.4
    Lifetime Mode=::NewEnumerator1
    Mass=&Lerp_Float(A=0.3, B=1, Alpha=$Particles.SparkIntensity)
    Mass Mode=::NewEnumerator1
    Sprite Size Mode=::NewEnumerator3
    Uniform Sprite Size=&Lerp_Float(A=1.5, B=2.5, Alpha=$Particles.SparkIntensity)
    UsePositionOffset=false
    Write Material Random=true
   )
   &SkeletalMeshLocation(
    Bone / Socket Index=$Particles.PartitionParticles.Partition
    Bone Sampling Mode=::NewEnumerator4
    MaskSpawnGroup=false
    Skeletal Mesh=$System.Skeletal Mesh
   )
   &AddVelocity(
    Cone Angle=7.5
    Cone Axis=&RotateVectorByQuaternion(
     Vector To Rotate=&SelectVectorFromArray(
      Direct Array Index=$Particles.SpawnGroup
      Vector Selection Array=DataInterface() {propertyValues {FloatData [<0.75 -0.5 0>, <0.35 0 0.8>, <-0.2 -0.8 0.3>]}}
     )
     Quaternion=$Particles.SkeletalMeshLocation.SampledRotation
    )
    Module.UseConeFalloff=false
    Module.UseVelDistribution=false
    Override Seed=false
    Velocity Mode=::NewEnumerator2
    Velocity Speed=&Multiply_Float(A=&RandomRangeFloat(Minimum=50, Maximum=150), B=2)
   ) {enabled=false}
   /SetVariables(Particles.SpriteSize=&Multiply_Vector2(
     A=$Particles.SpriteSize
     B=&Vector2DFromFloat(Value=&FloatFromCurve(FloatCurve=DataInterface() {propertyValues {Curve {keys [<at 1000 1 RCIM_Cubic>,<at 5000 2>]}}}, CurveIndex=$Engine.Owner.LODDistance))
    )) {enabled=false, isSetParameters=true}
   &ShapeLocation(MaskSpawnGroup=false, Override Seed=false, Sphere Radius=10, UseOwnerScale=true)
   &AddVelocityInCone(
    Cone Angle=7.5
    Cone Axis=&SelectVectorFromArray(
     Direct Array Index=$Particles.SpawnGroup
     Vector Selection Array=DataInterface() {propertyValues {FloatData [<0.75 -0.5 0>, <0.35 0 0.8>, <-0.2 -0.8 0.3>]}}
    )
    Module.Use Velocity Falloff On Cone Axis=false
    Velocity Strength=&RandomRangeFloat(Minimum=50, Maximum=150)
   ) {enabled=false}
   &AddVelocity(
    Module.Use Falloff Distance=false
    Override Seed=false
    Velocity Mode=::NewEnumerator1
    Velocity Origin=$Particles.SkeletalMeshLocation.SampledPosition
    Velocity Speed=&Multiply_Float(
     A=&RandomRangeFloat(Minimum=0.5, Maximum=1)
     B=&Multiply_Float(A=2000, B=1)
    )
   ) {enabled=false, module=AddVelocity001}
  ]
  particleUpdateScript [
   &ParticleState()
   &SkeletalMeshLocation(
    Bone / Socket Index=$Particles.PartitionParticles.Partition
    Bone Sampling Mode=::NewEnumerator4
    MaskSpawnGroup=false
    Orientation Sampling=::NewEnumerator1
    Position Sampling=::NewEnumerator2
    Sampled Position Offset=<0 0 5>
    Skeletal Mesh=$System.Skeletal Mesh
   ) {module=SkeletalMeshLocation001}
   &GravityForce(Gravity=&MakeVector(
     X=0
     Y=0
     Z=&FloatFromCurve(
      FloatCurve=DataInterface() {propertyValues {Curve {keys [{tangentMode=RCTM_User, time=0, value=0}, {tangentMode=RCTM_User, time=1, value=1}]}}}
      Scale Curve=500
     )
    ))
   &Drag(
    Drag=&FloatFromCurve(FloatCurve=DataInterface() {propertyValues {Curve {keys [<at 0 0.15>,<at 1 10>]}}})
    Use Linear Drag=true
    Use Rotational Drag=false
   ) {enabled=false}
   &Drag(Drag=&RandomRangeFloat(Minimum=4, Maximum=4.5), Use Linear Drag=true, Use Rotational Drag=false) {module=Drag001}
   &VectorNoiseForce(
    Force Amount=&FloatFromCurve(
     FloatCurve=DataInterface() {propertyValues {Curve {keys [
         {arriveTangent=-0.0392, interpMode=RCIM_Cubic, leaveTangent=-0.0392, tangentMode=RCTM_User, time=0, value=0.5}
         {arriveTangent=1.6186, interpMode=RCIM_Cubic, leaveTangent=1.6186, tangentMode=RCTM_User, time=0.6, value=0.975}
         <at 0.7 0 RCIM_Cubic>
        ]}}}
     Scale Curve=1500
    )
    Module.Use Random Noise Time Interval=false
   )
   &CurlNoiseForce(Bias Noise=false, Noise Frequency=15, Noise Strength=2500, Pan Noise=false) {enabled=false}
   &CurlNoiseForce(Bias Noise=false, Noise Strength=500, Pan Noise=false) {enabled=false, module=CurlNoiseForce001}
   &SolveForcesAndVelocity(Clamp Velocity=false, Limit Acceleration=false)
   /SetVariables(Particles.Velocity=&Add_Vector(
     A=$Particles.Velocity
     B=&Multiply_Vector(
      A=&Multiply_VectorByFloat(
       Vector=$Particles.SkeletalMeshLocation001.SampledVelocity
       Float=&FloatFromCurve(
        FloatCurve=DataInterface() {propertyValues {Curve {keys [<at 0 1>,<at 0.15 0>]}}}
        Scale Curve=&RandomRangeFloat(Minimum=1.5, Maximum=1.8)
       )
      )
      B=&RandomRangeVector(Minimum=<1.2 1.2 0.15>, Maximum=<1.4 1.4 0.35>)
     )
    )) {isSetParameters=true}
   &ScaleColor(
    Scale RGB=&VectorFromFloat(Value=&FloatFromCurve(
      FloatCurve=DataInterface() {propertyValues {Curve {keys [
          <at 0 2 RCIM_Cubic>
          {arriveTangent=-1.2206, interpMode=RCIM_Cubic, leaveTangent=-1.2206, tangentMode=RCTM_User, time=0.1, value=0.5}
          <at 0.75 0.1>
         ]}}}
      Scale Curve=8
     ))
    ScaleA=false
    ScaleRGB=true
   )
   &ScaleSpriteSize(Scale Factor=&Vector2DFromFloat(Value=&FloatFromCurve(FloatCurve=DataInterface() {propertyValues {Curve {keys [<at 0.4 1 RCIM_Cubic>, {arriveTangent=-3.8662, interpMode=RCIM_Cubic, leaveTangent=-3.8662, tangentMode=RCTM_User, time=1, value=0}]}}})))
   &ScaleSpriteSizeBySpeed(
    Max Scale Factor=<0.75 2>
    Min Scale Factor=<0.5 1>
    Sample Scale Factor By Curve=true
    Scale Factor Curve=DataInterface() {propertyValues {Curve {keys [
        {arriveTangent=0.2757, time=0, value=0}
        {arriveTangent=0.3844, interpMode=RCIM_Cubic, leaveTangent=0.3844, tangentMode=RCTM_User, time=0.3, value=0.15}
        {arriveTangent=2.2765, interpMode=RCIM_Cubic, leaveTangent=2.2765, tangentMode=RCTM_User, time=1, value=1}
       ]}}}
    Velocity Threshold=100
   )
   &SpriteFacingAndAlignment(Sprite Alignment=$Particles.Velocity, WriteAlignment=true, WriteFacing=false) {enabled=false}
  ]
  rendererClasses [&NiagaraSpriteRendererProperties]
  renderers [{rendererClass=&NiagaraSpriteRendererProperties}]
 }
}
summary {
 systemName=NS_Emote_PolishedJade_Despawn
 emitters [
  {emitterName=NE_FN_Sparks, simTarget=GPUComputeSim, rendererClasses [&NiagaraSpriteRendererProperties]}
  {emitterName=BlowingParticles001, simTarget=GPUComputeSim, rendererClasses [&NiagaraSpriteRendererProperties]}
  {emitterName=Default_Pickaxe_Impact_Smoke, rendererClasses [&NiagaraSpriteRendererProperties]}
 ]
 userVariables [
  User.bIsFrontEnd {type=NiagaraBool}
  User.Interpolate Intensity {type=NiagaraFloat, defaultValue=0.1, description=""}
  User.NormalizedNotifyProgress {type=NiagaraFloat, defaultValue=0, description=""}
  User.RibbonIDOffset {type=NiagaraInt32}
  User.TurnOff {type=NiagaraFloat, defaultValue=0, description=""}
 ]
}
