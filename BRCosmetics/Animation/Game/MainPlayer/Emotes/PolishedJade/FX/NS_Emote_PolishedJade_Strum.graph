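# Niagara system for the PolishedJade emote's strum effect, bursting short-lived jade-teal particles from a skeletal mesh.
# BlowingParticles001 spawns a brief glow with sprite and light renderers, colored by time of day.
# DirectionalBurst scatters small spinning motes with curl noise, drag, random visibility tags and randomly offset brightness spikes for twinkle.
# Both emitters sample the system-level skeletal mesh, inheriting its velocity early in particle life so particles follow the motion.
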NiagaraSystem /BRCosmetics/Animation/Game/MainPlayer/Emotes/PolishedJade/FX/NS_Emote_PolishedJade_Strum
dependencies {
 usedDataInterfaces [&NiagaraDataInterfaceCurve, &NiagaraDataInterfaceSkeletalMesh]
 usedDynamicInputs []
 usedModules [
  &SystemState
  &self:SystemSpawnScript.SystemScriptSource.SystemScriptGraph.NiagaraNodeAssignment_0.SetVariables_#1
 ]
}
emitters {
 BlowingParticles001 {
  emitterUpdateScript [
   &EmitterState()
   &SpawnBurst_Instantaneous(Use Spawn Probability=false, UseLoopCountLimit=false)
   &SpawnRate(SpawnRate=15, Use Spawn Probability=false) {enabled=false}
  ]
  particleSpawnScript [
   &InitializeParticle(
    Color=&FN_Select_Color_from_Time_Of_Day(Morning=<rgba 0.2203 0.9 0.7754 1>, Day=<rgba 0.2203 0.9 0.7754 1>, Evening=<rgba 0.2203 0.9 0.7754 1>, Night=<rgba 0.2203 0.7641 0.9 1>)
    Lifetime=0.2
    Mass Max=1.5
    Mass Mode=::NewEnumerator3
    Sprite Rotation Mode=::NewEnumerator2
    Sprite Size Mode=::NewEnumerator3
    Uniform Sprite Size=&RandomRangeFloat(Minimum=25, Maximum=30)
    UsePositionOffset=false
    Write Material Random=true
   )
   &SkeletalMeshLocation(MaskSpawnGroup=false, Mesh Sampling Type=::NewEnumerator5, Skeletal Mesh=$System.Skeletal Mesh, Socket Sampling Mode=::NewEnumerator5)
   &AddVelocity(Velocity=<0 0 -1>, Velocity Speed Scale=&RandomRangeFloat(Minimum=5, Maximum=20)) {module=AddVelocity001}
   &CameraOffset(Camera Offset Amount=5)
  ]
  particleUpdateScript [
   &ParticleState()
   &SkeletalMeshLocation(
    Bone Sampling Mode=::NewEnumerator4
    "Inherit Velocity (Scale)":NiagaraFloat=&FloatFromCurve(
     FloatCurve=DataInterface() {propertyValues {Curve {keys [{tangentMode=RCTM_User, time=0, value=1}, {tangentMode=RCTM_User, time=0.1, value=0}]}}}
     Scale Curve=&RandomRangeFloat(Minimum=0.2, Maximum=0.3)
    )
    MaskSpawnGroup=false
    Position Sampling=::NewEnumerator2
    Skeletal Mesh=$System.Skeletal Mesh
    Velocity Sampling=::NewEnumerator0
    VelocitySpeedLimit=false
   ) {module=SkeletalMeshLocation001}
   &ScaleSpriteSize(Uniform Curve Sprite Scale=DataInterface() {propertyValues {Curve {keys [
        {arriveTangent=-0.1114, interpMode=RCIM_Cubic, leaveTangent=-0.1114, tangentMode=RCTM_User, time=0, value=0}
        {arriveTangent=2.1874, interpMode=RCIM_Cubic, leaveTangent=2.1874, tangentMode=RCTM_User, time=1, value=1}
       ]}}}) {enabled=false}
   &ScaleColor(
    Scale Alpha=&FloatFromCurve(
     FloatCurve=DataInterface() {propertyValues {Curve {keys [
         {arriveTangent=-0.0725, interpMode=RCIM_Cubic, leaveTangent=-0.0725, tangentMode=RCTM_User, time=0, value=0}
         {arriveTangent=0.1465, interpMode=RCIM_Cubic, leaveTangent=0.1465, tangentMode=RCTM_User, time=0.25, value=0.9847}
         {arriveTangent=-0.0405, interpMode=RCIM_Cubic, leaveTangent=-0.0405, tangentMode=RCTM_User, time=1, value=0}
        ]}}}
     Scale Curve=0.25
    )
    Scale RGB=&VectorFromFloat(Value=1)
    ScaleA=true
    ScaleRGB=true
   )
   &FN_Sprite_Attributes(
    Color=&MultiplyLinearColor(
     Linear Color=$Particles.Color
     Scale Factor=&MakeFloat4FromVectorAndFloat("Vector (RGB)"=&VectorFromFloat(Value=0.05), "Float (Alpha)"=1)
    )
    Create Color Attribute=true
    Create SpriteSize Attribute=true
    SpriteSize=&Multiply_Vector2D_ByFloat(Float=3, Vector2D=$Particles.SpriteSize)
    Write to Module Namespace=true
   )
   &FN_Light_Attributes(
    Color=$Particles.Color
    Create Color Attribute=true
    Create Exponent Attribute=true
    Create Position Attribute=true
    Create Radius Attribute=true
    Enable Particle Light Enable=true
    Exponent=1
    Light Enabled=true
    Position:NiagaraPosition=$Particles.Position
    Radius=15
    Scale Color=&FloatFromCurve(
     FloatCurve=DataInterface() {propertyValues {Curve {keys [{interpMode=RCIM_Cubic, tangentMode=RCTM_User, time=0, value=1}, {interpMode=RCIM_Cubic, tangentMode=RCTM_User, time=1, value=0}]}}}
     Scale Curve=2
    )
   )
   &Drag(Drag=&RandomRangeFloat(Minimum=4, Maximum=5), Use Linear Drag=true, Use Rotational Drag=false) {module=Drag001}
   &SolveForcesAndVelocity(Clamp Velocity=false, Limit Acceleration=false)
   &DynamicMaterialParameters(
    Index 0 Param 1=3
    Index 0 Param 2=1
    Module.Index0Param0Write=true
    Module.Index0Param1Write=true
    Module.Index0Param2Write=true
    Module.Index0Param3Write=true
   )
  ]
  rendererClasses [&NiagaraSpriteRendererProperties, &NiagaraLightRendererProperties]
  renderers [
   {rendererClass=&NiagaraSpriteRendererProperties}
   {rendererClass=&NiagaraSpriteRendererProperties, rendererIndex=1}
   {rendererClass=&NiagaraSpriteRendererProperties, rendererIndex=2}
   {rendererClass=&NiagaraLightRendererProperties, rendererIndex=3}
  ]
 }
 DirectionalBurst {
  emitterUpdateScript [
   &EmitterState()
   &SpawnBurst_Instantaneous(
    Spawn Count=&RandomRangeInt(Minimum=8, Maximum=12)
    Use Spawn Probability=false
    UseLoopCountLimit=false
   )
  ]
  particleSpawnScript [
   &InitializeParticle(
    Color=&RandomRangeLinearColor(Minimum=<rgba 0.0018 0.0885 0.0857 1>, Maximum=<rgba 0.2203 0.9 0.7754 1>)
    Lifetime Max=0.8
    Lifetime Min=0.3
    Lifetime Mode=::NewEnumerator1
    Mass Max=2
    Mass Mode=::NewEnumerator3
    Sprite Size Mode=::NewEnumerator3
    Uniform Sprite Size=&RandomRangeFloat(Minimum=5, Maximum=10)
    UsePositionOffset=false
    Write Material Random=true
   )
   &AddVelocity(
    Cone Angle=50
    Distribution Along Cone Axis=0.2
    Module.UseConeFalloff=false
    Module.UseVelDistribution=true
    Override Seed=false
    Velocity Mode=::NewEnumerator2
    Velocity Speed=&RandomRangeFloat(Minimum=250, Maximum=750)
   ) {enabled=false}
   &SkeletalMeshLocation(
    "Inherit Velocity (Scale)"=100
    MaskSpawnGroup=false
    Mesh Sampling Type=::NewEnumerator5
    Skeletal Mesh=$System.Skeletal Mesh
    Socket Sampling Mode=::NewEnumerator5
    Velocity Sampling=::NewEnumerator0
    VelocitySpeedLimit=false
   )
   &ShapeLocation(MaskSpawnGroup=false, Override Seed=false, Sphere Radius=3, UseOwnerScale=true)
   /SetVariables(Particles.VisibilityTag=&MakeIntFromFloat(Float=&FloatFromCurve(
      FloatCurve=DataInterface() {propertyValues {Curve {keys [<at 0 0.5 RCIM_Constant>,<at 0.95 1.5 RCIM_Constant>]}}}
      CurveIndex=&RandomRangeFloat(Minimum=0, Maximum=1)
     ))) {isSetParameters=true}
   &AddVelocity(Velocity=<0 0 -1>, Velocity Speed Scale=&RandomRangeFloat(Minimum=0, Maximum=200)) {module=AddVelocity001}
  ]
  particleUpdateScript [
   &ParticleState()
   &ScaleSpriteSize(
    Scale Sprite Size Mode=::NewEnumerator0
    Uniform Scale Factor=&FloatFromCurve(FloatCurve=DataInterface() {propertyValues {Curve {keys [
         {arriveTangent=-0.0089, interpMode=RCIM_Cubic, leaveTangent=-0.0089, tangentMode=RCTM_User, time=0, value=0}
         {arriveTangent=-2.5249, interpMode=RCIM_Cubic, leaveTangent=-2.5249, tangentMode=RCTM_User, time=0.1408, value=1}
         {arriveTangent=-0.0405, interpMode=RCIM_Cubic, leaveTangent=-0.0405, tangentMode=RCTM_User, time=1, value=0}
        ]}}})
   )
   &SpriteRotationRate(Rotation Rate=&Multiply_Float(
     A=&UniformAOrBFloat(A=-1, Evaluation Type=::NewEnumerator0, Recalculate Random Each Loop=true, Override Seed=false)
     B=&RandomRangeFloat(Minimum=240, Maximum=480)
    ))
   &SkeletalMeshLocation(
    Bone Sampling Mode=::NewEnumerator4
    "Inherit Velocity (Scale)":NiagaraFloat=&FloatFromCurve(
     FloatCurve=DataInterface() {propertyValues {Curve {keys [{tangentMode=RCTM_User, time=0, value=1}, {tangentMode=RCTM_User, time=0.15, value=0}]}}}
     Scale Curve=&RandomRangeFloat(Minimum=0.2, Maximum=0.45)
    )
    MaskSpawnGroup=false
    Position Sampling=::NewEnumerator2
    Sampled Position Offset=&RandomRangeVector(Minimum=<0 -5 0>, Maximum=<0 -3 0>)
    Skeletal Mesh=$System.Skeletal Mesh
    Velocity Sampling=::NewEnumerator0
    VelocitySpeedLimit=false
   ) {module=SkeletalMeshLocation001}
   &GravityForce() {enabled=false}
   &Drag(Drag=&RandomRangeFloat(Minimum=0.8, Maximum=1.2), Use Linear Drag=true, Use Rotational Drag=false) {enabled=false}
   &ScaleColor(
    Scale Alpha=&FloatFromCurve(FloatCurve=DataInterface() {propertyValues {Curve {keys [<at -0.0046 1.0113 RCIM_Cubic>,<at 1.0011 0.002 RCIM_Cubic>]}}})
    ScaleA=true
    ScaleRGB=true
   ) {enabled=false}
   &Drag(Drag=&RandomRangeFloat(Minimum=4, Maximum=5), Use Linear Drag=true, Use Rotational Drag=false) {module=Drag001}
   &CurlNoiseForce(Bias Noise=false, Noise Frequency=250, Noise Strength=150, Pan Noise=false)
   &SolveForcesAndVelocity(Clamp Velocity=false, Limit Acceleration=false)
   &ScaleSpriteSizeBySpeed(
    Max Scale Factor=<1.2 3.3>
    Min Scale Factor=<0.75 0.3>
    Sample Scale Factor By Curve=true
    Scale Factor Curve=DataInterface() {propertyValues {Curve {keys [
        {arriveTangent=2.865, interpMode=RCIM_Cubic, time=0.1247, value=-0.0006}
        {arriveTangent=4.1522, interpMode=RCIM_Cubic, leaveTangent=4.1522, tangentMode=RCTM_User, time=0.3948, value=0.4893}
        <at 0.9958 0.9974 RCIM_Cubic>
       ]}}}
    Velocity Threshold=1500
   ) {enabled=false}
   &ScaleColor(
    Scale Alpha=&FloatFromCurve(FloatCurve=DataInterface() {propertyValues {Curve {keys [
         {arriveTangent=0.088, interpMode=RCIM_Cubic, leaveTangent=0.088, tangentMode=RCTM_User, time=0, value=0}
         {arriveTangent=-2.5249, interpMode=RCIM_Cubic, leaveTangent=-2.5249, tangentMode=RCTM_User, time=0.1408, value=0.9847}
         {arriveTangent=-0.0405, interpMode=RCIM_Cubic, leaveTangent=-0.0405, tangentMode=RCTM_User, time=1, value=0}
        ]}}})
    Scale RGB=&VectorFromFloat(Value=&Add_Float(
      A=&FloatFromCurve(FloatCurve=DataInterface() {propertyValues {Curve {keys [
           {arriveTangent=-0.0052, interpMode=RCIM_Cubic, leaveTangent=-0.0052, tangentMode=RCTM_User, time=0, value=1}
           {arriveTangent=-2.5249, interpMode=RCIM_Cubic, leaveTangent=-2.5249, tangentMode=RCTM_User, time=0.1408, value=3.5}
           {interpMode=RCIM_Cubic, leaveTangent=-0.0405, time=1, value=1}
          ]}}})
      B=&FloatFromCurve(
       FloatCurve=DataInterface() {propertyValues {Curve {keys [
           {arriveTangent=-0.0052, interpMode=RCIM_Cubic, leaveTangent=-0.0052, tangentMode=RCTM_User, time=0.4, value=0}
           {arriveTangent=6292.0376, interpMode=RCIM_Cubic, leaveTangent=-4742.8989, tangentMode=RCTM_Break, time=0.5, value=350}
           {interpMode=RCIM_Cubic, leaveTangent=6.25, time=0.6, value=0}
          ]}}}
       CurveIndex=&ModuloFloat(
        A=1
        B=&Add_Float(A=$Particles.NormalizedAge, B=&RandomRangeFloat(Minimum=0, Maximum=1))
       )
      )
     ))
    ScaleA=true
    ScaleRGB=true
   ) {module=ScaleColor001}
   &FN_Sprite_Attributes(
    Create SpriteSize Attribute=true
    SpriteSize=&Multiply_Vector2D_ByFloat(Float=1.5, Vector2D=$Particles.SpriteSize)
    Write to Module Namespace=true
   )
  ]
  rendererClasses [&NiagaraSpriteRendererProperties]
  renderers [
   {rendererClass=&NiagaraSpriteRendererProperties}
   {rendererClass=&NiagaraSpriteRendererProperties, rendererIndex=1}
   {rendererClass=&NiagaraSpriteRendererProperties, rendererIndex=2}
  ]
 }
}
summary {
 systemName=NS_Emote_PolishedJade_Strum
 emitters [
  {emitterName=DirectionalBurst, rendererClasses [&NiagaraSpriteRendererProperties]}
  {
   emitterName=BlowingParticles001
   rendererClasses [&NiagaraSpriteRendererProperties, &NiagaraLightRendererProperties]
  }
 ]
 userVariables []
}
