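# Translucent unlit material for the Polished Jade emote sprite, built for Niagara mesh particles.
# Particle color combines with atmospheric light and a tinted, scaled glow from an eye-glow texture.
# A view-angle mask and instance-bounds gradient shape a fade, alongside time-of-day brightness and player and camera distance masks.
# Sine undulation along normals, offset by per-particle random and weighted by vertex color, feeds a WPO reroute.
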Material /BRCosmetics/Animation/Game/MainPlayer/Emotes/PolishedJade/Materials/M_Emote_PolishJade_Sprite
graph {
 properties {
  assetImportData="<self>:InterchangeAssetImportData_0"
  bUsedWithInstancedStaticMeshes=true
  bUsedWithNiagaraMeshParticles=true
  blendMode=BLEND_Translucent
  shadingModel=MSM_Unlit
  usedShadingModels=Unlit
 }
 nodes [
  n1 = DynamicParameter() {parameterIndex=1, paramNames [X,Y,Z,Param4]}
  n2 = MaterialFunctionCall(X=@n1.X, Y=@n1.X, Z=@n1.X) {materialFunction=&MakeFloat3}
  n3 = MaterialFunctionCall() {materialFunction=/Game/Effects/Niagara/MasterMaterials/Functions/InstanceBounds_0-1_UVW}
  n4 = VertexColor()
  n5 = Add(
   A=Multiply(A=OneMinus(None=@n3.B), B=ScalarParameter() {defaultValue=0.25, parameterName=Body Movement})
   B=Multiply(A=Power(Base=@n4.R), B=ScalarParameter() {defaultValue=0.25, parameterName=Param_2})
  )
  n6 = Multiply(
   A=LinearInterpolate(B=Transform(None=@n2.Result) {transformSourceType=TRANSFORMSOURCE_World, transformType=TRANSFORM_Instance}, Alpha=Saturate(None=@n5))
   B=ScalarParameter() {defaultValue=1, parameterName=Secondary Amount}
  )
  Add(A=@n6) {constB=0}
  n7 = Time()
  n8 = ScalarParameter() {defaultValue=1, parameterName=Undulation}
  n9 = Add(
   A=MaterialFunctionCall() {materialFunction=&InstanceBounds_0-1_UVW}.B
   B=Add(A=ParticleRandom(), B=Multiply(A=@n7, B=@n8))
  )
  n10 = Multiply(A=VectorParameter() {parameterName=Param_3, defaultValue=<rgba 0.75 0.75 0 0>}.RGB, B=Multiply(A=VertexNormalWS(), B=Sine(None=@n9)))
  Add(A=@n6, B=@n10)
  n11 = Saturate(None=DotProduct(A=AtmosphericLightVector(), B=VertexNormalWS()))
  n12 = ParticleColor()
  n13 = ScalarParameter() {parameterName=Param_8}
  n14 = ParticleColor()
  n15 = VectorParameter() {parameterName=Color, defaultValue=<rgba 2 1.3092 0.625 1>}
  n16 = ScalarParameter() {defaultValue=10, parameterName=Color Scale}
  n17 = Multiply(
   A=TextureSampleParameter2D() {parameterName=EmissiveColorMap, texture=/BRCosmetics/Animation/Game/MainPlayer/Emotes/PolishedJade/Textures/T_EnvSprite_EyeGlows}.G
   B=NamedRerouteUsage()
  )
  n18 = Multiply(A=Multiply(A=Multiply(A=@n15.RGB, B=@n16), B=@n17), B=Saturate(None=@n14.A))
  Add(A=@n11, B=Add(A=Multiply(A=@n12.RGB, B=@n13), B=Add(A=@n14.RGB, B=@n18)))
  Add(A=@n2.Result, B=WorldPosition().XYZ)
  Constant() {r=1}
  n19 = CollectionParameter() {collection=&FortniteMaterialParameters, parameterName=PlayerPosition}
  n20 = ComponentMask(None=@n19) {b=true, g=true, r=true}
  n21 = ParticlePositionWS()
  n22 = ScalarParameter() {defaultValue=750, parameterName=Eye show distance}
  LinearInterpolate(
   A=ScalarParameter() {parameterName=Param_9}
   B=ScalarParameter() {defaultValue=0.25, parameterName=Bokeh Fade}
   Alpha=OneMinus(None=SphereMask(A=@n20, B=@n21, Radius=@n22) {hardnessPercent=35})
  )
  n23 = Multiply(A=@n12.RGB) {constB=1.5}
  MaterialFunctionCall(Emissive=Multiply(A=LinearInterpolate(A=@n12.RGB, B=@n23, Alpha=@n11), B=@n12.A)) {materialFunction=&MF_FX_EmissiveAdjustments}
  MaterialFunctionCall(Brightness=Constant() {r=1}, Exponent=Constant() {r=5}) {materialFunction=&MF_Fresnel_FX}
  n24 = ScalarParameter() {defaultValue=1, parameterName=Param_1}
  n25 = DotProduct(A=CameraVectorWS(), B=VertexNormalWS())
  n26 = Multiply(
   A=SphereMask(
    A=Saturate(None=@n25)
    B=ScalarParameter() {defaultValue=0.688, parameterName=Mask Start}
    Radius=ScalarParameter() {defaultValue=0.336, parameterName=Mask Radius}
    Hardness=ScalarParameter() {parameterName=Mask Hardness}
   )
   B=ParticleColor().A
  )
  n27 = MaterialFunctionCall() {materialFunction=&InstanceBounds_0-1_UVW}
  n28 = ScalarParameter() {defaultValue=0.75, parameterName=Param_5}
  Multiply(A=Add(A=Multiply(A=@n17), B=Multiply(A=@n24, B=@n26)), B=OneMinus(None=Power(Base=@n27.B, Exp=@n28)))
  Multiply(A=MaterialFunctionCall(Float2=TextureCoordinate() {coordinateIndex=1}) {materialFunction=&BreakOutFloat2Components}.R)
  n29 = ParticlePositionWS()
  n30 = WorldPosition()
  n31 = Time()
  n32 = VectorParameter() {parameterName=Pan Speed, defaultValue=<rgba 0 0 -0.25 0>}
  n33 = Multiply(A=Add(A=Subtract(A=@n29, B=@n30.XYZ), B=Multiply(A=@n31, B=@n32.RGB)), B=ScalarParameter() {defaultValue=2, parameterName=WPO Noise Scale})
  Multiply(
   A=Multiply(A=VertexNormalWS(), B=VectorParameter() {parameterName=Offset Direction, defaultValue=<rgba 0 0 20 0>}.RGB)
   B=Multiply(A=VertexColor().G, B=TextureSample(UVs=@n33) {samplerType=SAMPLERTYPE_LinearColor, texture=/Game/Global/MaterialFunctions/VT_ClothSimWPO}.G)
  )
  n34 = LinearInterpolate(Alpha=OneMinus(None=VertexColor().R)) {constA=0.35}
  NamedRerouteDeclaration(None=Multiply(A=Multiply(A=@n10, B=@n34), B=ScalarParameter() {defaultValue=1, parameterName=WPO Mult})) {name=WPO, nodeColor=<rgba 0 0.6016 1 1>}
  n35 = SphereMask(A=WorldPosition().XYZ, B=CameraPositionWS(), Radius=ScalarParameter() {defaultValue=2650, parameterName=Param_4}) {attenuationRadius=5500, hardnessPercent=0}
  NamedRerouteDeclaration(None=Power(Base=OneMinus(None=@n35), Exp=ScalarParameter() {defaultValue=20, parameterName=Param_3})) {name=Distance, nodeColor=<rgba 0.6406 1 0 1>}
  NamedRerouteDeclaration(None=MaterialFunctionCall(
    Value_Morning=ScalarParameter() {defaultValue=1.5, parameterName=Opacity Morning}
    Value_Day=ScalarParameter() {defaultValue=2, parameterName=Opacity Day}
    Value_Evening=ScalarParameter() {defaultValue=1.5, parameterName=Opacity Evening}
    Value_Night=ScalarParameter() {defaultValue=2, parameterName=Opacity Night}
   ) {materialFunction=&Scalar_Time_of_Day}.Result) {name=TOD Brightness, nodeColor=<rgba 0 0.9766 1 1>}
  NamedRerouteUsage()
 ]
}
